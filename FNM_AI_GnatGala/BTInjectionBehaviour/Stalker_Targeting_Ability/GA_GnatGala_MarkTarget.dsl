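; Gameplay ability for the GnatGala AI hunter that marks its current perception target with a stalker effect.
; On activation it commits, caches the avatar pawn as HunterActor and binds to the bot controller's target-changed event.
; A new target receives the apply effect; only losing the target sends the remove effect, so switching leaves the old one untouched.
; On end, temporary stalk-target UI starts fading out but is removed immediately.
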(event Ability|EventOnEndAbility
  (params (bWasCancelled :type Boolean))
  ; TEMP - UI for Stalking/Marked Target
  (bind _asw_stalk_target_ui (Utilities|Casting|CastToW_StalkTarget_UI (Variables|Default|GetUser%20Widget)))
  (UserInterface|Animation|PlayAnimation _asw_stalk_target_ui (Variables|/TidalCranePrototype/Gameplay/StalkerOnTheLoose/UI/W_StalkTarget_UI.W_StalkTarget_UI_C|GetFadeOut _asw_stalk_target_ui) 0.0 1 "Forward" 1.0 false)
  (bind _user_widget (Variables|Default|GetUser%20Widget)
    (:then
      (Widget|RemovefromParent _user_widget)
      (|CallParentFunction bWasCancelled)
      (Ability|EndAbility))
    (:else)))

(event Ability|EventActivateAbility
  (Ability|CommitAbility)
  (bind _asfort_player_pawn (Utilities|Casting|CastToFortPlayerPawn (Ability|GetAvatarActorfromActorInfo)))
  (Variables|Default|SetHunterActor _asfort_player_pawn)
  (bind _asfort_athena_aibot_controller (Utilities|Casting|CastToFortAthenaAIBotController (Pawn|GetController _asfort_player_pawn)))
  (Perception|BindEventtoOnCurrentTargetChangedEvent _asfort_athena_aibot_controller (delegate Custom|TargetChanged)))

(event Custom|TargetChanged
  (params (OldTarget :type Actor%20Object%20Reference) (NewTarget :type Actor%20Object%20Reference))
  (if (== NewTarget 0)
    (GameplayEffects|ApplyGameplayEffectToTarget :self (Ability|GetAbilitySystemComponent (Variables|Default|GetHunterActor)) :GameplayEffectClass (Variables|Default|GetRemoveStalkerEffectGE) :Target (Ability|GetAbilitySystemComponent NewTarget) :Level 0.0)
    (GameplayEffects|ApplyGameplayEffectToTarget :self (Ability|GetAbilitySystemComponent (Variables|Default|GetHunterActor)) :GameplayEffectClass (Variables|Default|GetRemoveStalkerEffectGE) :Target (Ability|GetAbilitySystemComponent OldTarget) :Level 0.0)
    (else
      (GameplayEffects|ApplyGameplayEffectToTarget :self (Ability|GetAbilitySystemComponent (Variables|Default|GetHunterActor)) :GameplayEffectClass (Variables|Default|GetApplyStalkerEffectGE) :Target (Ability|GetAbilitySystemComponent NewTarget) :Level 0.0))))
